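# Conversation database for the NPC Mizuki of the River Guard, defining the greeting and top-level options when a player talks to this character.
# The entry point feeds one speech node, which picks a contextual message by requirement and priority: first-ever meeting, match conversation count, or the player's character cosmetic, with a default fallback.
# The speech node then branches through link nodes to the shared NPCLibrary quests and basic services entries, plus Mizuki's own dialogue entry.
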ConversationDatabase /NPCLibrary/NPCs/Mizuki/NPCLibrary_Conversation_Mizuki
entry_points [{tag=Conversation.Entry.Mizuki, destinations [#1]}]
nodes [
 ConversationEntryPointNode_7 {
  id=#1
  class=ConversationEntryPointNode
  outputs [#2]
  properties {entryTag {tagName=Conversation.Entry.Mizuki}}
 }
 Speech_C_4 {
  id=#2
  class=Speech_C
  outputs [#3, #4, #5]
  properties {
   nodeName=""
   generalConfig {
    defaultMessage=War has carved its path through this Island before. History is a cycle.
    contextualMessages [
     {
      message="Hello. I am Mizuki, empress of the River Guard. "
      contextRequirements [{requirement {class=FortControllerRequirement_IsFirstEverConversationWithNPC}}]
      priority {curve {rowName=Tandem.Speech.Priority.FirstEverInteraction}}
     }
     {
      message=Everything in flow.
      contextRequirements [{requirement {class=FortControllerRequirement_CheckNPCMatchConversationCount, properties {nPCUniqueIdTag {tagName=AISpawnerData.Type.Tandem.Mizuki}}}}]
      priority {curve {rowName=None}}
     }
     {
      message=War has carved its path through this Island before. History is a cycle.
      contextRequirements [{requirement {class=FortControllerRequirement_CheckNPCMatchConversationCount, properties {nPCUniqueIdTag {tagName=AISpawnerData.Type.Tandem.Mizuki}}}}]
      priority {curve {rowName=None}}
     }
     {
      message="Hail, from the River Guard."
      contextRequirements [{requirement {class=FortControllerRequirement_CheckNPCMatchConversationCount, properties {nPCUniqueIdTag {tagName=AISpawnerData.Type.Tandem.Mizuki}}}}]
      priority {curve {rowName=None}}
     }
     {
      message=Your plans are naive and foolish... but River Guard stands with you.
      contextRequirements [{requirement {
         class=FortControllerRequirement_HasCID
         properties {allowedCIds [
           &Character_IronLilac_Sly
           &Character_UndergroundRebel_Fashion
           &CID_989_Athena_Commando_M_ProgressiveJonesy
           &Character_AuraCop_Bullet
           &Character_BrightDisk
           &Character_VoidRedemption_Rebel
          ]}
        }}]
      priority {value=2, curve {curveTable=null, rowName=None}}
     }
     {
      message=The echoes of your ancestor's actions still ripple outward. What will your legacy be?
      contextRequirements [{requirement {
         class=FortControllerRequirement_HasCID
         properties {allowedCIds [
           /BRCosmetics/Athena/Items/Cosmetics/Characters/Character_PolishedJade_Mind
           &Character_CubicVice_Dual
           &Character_MistMylar_Claw
           /BRCosmetics/Athena/Items/Cosmetics/Characters/Character_SolidMist
          ]}
        }}]
      priority {value=2, curve {curveTable=null, rowName=None}}
     }
     {
      message=The Fox Clan have my respect. Stay safe.
      contextRequirements [{requirement {
         class=FortControllerRequirement_HasCID
         properties {allowedCIds [
           /BRCosmetics/Athena/Items/Cosmetics/Characters/CID_161_Athena_Commando_M_Drift
           /BRCosmetics/Athena/Items/Cosmetics/Characters/CID_461_Athena_Commando_M_DriftSummer
           /BRCosmetics/Athena/Items/Cosmetics/Characters/CID_486_Athena_Commando_F_StreetRacerDrift
           /BRCosmetics/Athena/Items/Cosmetics/Characters/Character_PreciseRuffian
           /BRCosmetics/Athena/Items/Cosmetics/Characters/CID_A_009_Athena_Commando_F_FoxWarrior_21B9R
           /BRCosmetics/Athena/Items/Cosmetics/Characters/CID_A_185_Athena_Commando_M_Polygon
           /BRCosmetics/Athena/Items/Cosmetics/Characters/Character_SunbeamQuest
          ]}
        }}]
      priority {value=2, curve {curveTable=null, rowName=None}}
     }
     {
      message="Be careful, brother. This war has deep roots, and is fueled by an ancient hunger."
      contextRequirements [{requirement {class=FortControllerRequirement_HasCID, properties {allowedCIds [/BRCosmetics/Athena/Items/Cosmetics/Characters/Character_Ebony]}}}]
      priority {value=2, curve {curveTable=null, rowName=None}}
     }
    ]
   }
  }
 }
 ConversationLinkNode_3 {id=#5, class=ConversationLinkNode, properties {remoteEntryTag {tagName=Conversation.Entry.Mizuki.Dialogue}}}
 ConversationLinkNode_0 {id=#4, class=ConversationLinkNode, properties {remoteEntryTag {tagName=Conversation.Entry.BasicServices}}}
 ConversationLinkNode_1 {id=#3, class=ConversationLinkNode, properties {remoteEntryTag {tagName=Conversation.Entry.NPCLibrary.Quests}}}
]
